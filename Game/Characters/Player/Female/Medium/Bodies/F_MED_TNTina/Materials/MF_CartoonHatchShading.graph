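# Cartoon hatch-shading material function for the TNTina character, replacing the input material's lit response with stylized shading written to emissive.
# Sun direction against blended pixel/vertex normals drives height lerps between crosshatch and halftone textures, projected on an FOV-scaled virtual camera plane with time-stepped hatch jitter.
# Results tint base color with time-of-day adjustments; an optional masked superlevel look adds cubemap reflections and gold tint.
# Several disconnected experimental branches remain.
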MaterialFunction /Game/Characters/Player/Female/Medium/Bodies/F_MED_TNTina/Materials/MF_CartoonHatchShading
graph {nodes [
  Constant2Vector() {g=0.5, r=0.5}
  Constant3Vector() {constant=<rgba 0.671 0.8701 1.1 1>}
  Constant3Vector() {constant=<rgba 0.6 0.6 3 1>}
  Constant() {r=2}
  Constant() {r=0.8}
  Constant() {r=0.2}
  Constant() {r=0.7}
  Constant() {r=5}
  Constant() {r=2}
  Constant() {r=1.5}
  Constant() {r=0.8}
  Constant() {r=4}
  Constant() {r=0.01}
  Constant() {r=1.4}
  Constant() {r=2}
  Constant() {r=1.75}
  Constant() {r=500}
  Constant() {r=6}
  Constant() {r=2}
  Constant() {r=4}
  Constant() {r=8.3}
  Constant() {r=12}
  Constant() {r=0.46}
  Constant() {r=1.25}
  Constant() {r=2}
  Constant() {r=8}
  Constant() {r=-2.5}
  n1 = Constant()
  n2 = GetMaterialAttributes(None=FunctionInput() {inputName=Material, inputType=FunctionInput_MaterialAttributes}) {attributeGetTypes [#1, #2]}
  n3 = ComponentMask(None=PixelNormalWS()) {b=true, g=true, r=true}
  n4 = VertexNormalWS()
  n5 = ShadingPathSwitch(
   Default=LinearInterpolate(
    A=@n3
    B=ComponentMask(None=@n4) {b=true, g=true, r=true}
    Alpha=ScalarParameter() {defaultValue=0.25, group=Cartoon, parameterName=CRTN_Pixel-Vertex}
   )
   Forward=@n3
   Mobile=@n3
  )
  n6 = TextureSample!(
   UVs=ReflectionVectorWS(CustomWorldNormal=@n5)
   Tex
   Level=ScalarParameter() {defaultValue=2, group=Cartoon, parameterName=CRTN_SL-ReflectMipLevel}
   Apply View MipBias
  ) {mipValueMode=TMVM_MipLevel, texture=/Game/Packages/Fortress_Sky/SkyDome/MaterialInstances/SkyDome_Apollo/Cubemap_Apollo_Day}
  n7 = Constant() {r=1}
  n8 = Multiply(
   A=Power(Base=Desaturation(None=@n6.RGB, Fraction=@n7), Exp=Constant() {r=6})
   B=ScalarParameter() {defaultValue=600, group=Cartoon, parameterName=CRTN_SL-ReflectMod}
  )
  n9 = Add(A=ScalarParameter() {defaultValue=-2, group=Cartoon, parameterName=CRTN_SL-ReflectLitAdd}, B=@n8)
  n10 = ScalarParameter() {defaultValue=1.5, group=Cartoon, parameterName=CRTN_SL-LitReflectMod}
  n11 = Constant() {r=0.5}
  n12 = CollectionParameter() {
   collection=&FortniteMaterialParameters
   parameterName=SunlightDirectionalVector
  }
  n13 = Subtract(A=DotProduct(A=ComponentMask(None=@n12) {b=true, g=true, r=true}, B=@n5)) {constB=-1}
  n14 = Saturate(None=Divide(A=@n13, B=ScalarParameter() {defaultValue=1.25, group=Cartoon, parameterName=CRTN_LightLevels}))
  n15 = Power(Base=@n14, Exp=ScalarParameter() {defaultValue=5, group=Cartoon, parameterName=CRTN_LightPower})
  n16 = ScalarParameter() {defaultValue=8.3, group=Cartoon, parameterName=CRTN_Crosshatch-OffsetUV}
  n17 = Multiply(A=ScalarParameter() {defaultValue=0.46, group=Cartoon, parameterName=CRTN_Crosshatch-OffsetSpeed}, B=Time())
  n18 = Floor(None=Multiply(A=ScalarParameter() {defaultValue=12, group=Cartoon, parameterName=CRTN_Crosshatch-NumberOffsets}, B=Frac(None=@n17)))
  n19 = ScalarParameter() {defaultValue=0.01, group=Cartoon, parameterName=CRTN_Crosshatch-Scale}
  n20 = Constant() {desc=Needs at least a small amount of this or it will break, r=0.25}
  n21 = MaterialFunctionCall() {materialFunction=&CameraDirectionVector}
  n22 = Normalize(VectorInput=Multiply(A=Constant3Vector() {constant=<rgba 1 1 0 0>}, B=@n21.Result))
  n23 = ScalarParameter() {defaultValue=70, group=Cartoon, parameterName=CRTN_VP-CameraVectorMin}
  n24 = ObjectPositionWS()
  n25 = Add(A=ActorPositionWS(), B=VectorParameter() {group=Cartoon, parameterName=CRTN_VP-ActorPosOffset, defaultValue=<rgba 0 0 80 1>}.RGB)
  n26 = MaterialFunctionCall(
   Input=MaterialFunctionCall() {materialFunction=/Engine/Functions/Engine_MaterialFunctions02/Utility/FOV}."FOV "
   Input Low=Constant() {r=0.414}
   Input High=Constant() {r=1.091}
   Target Low=Constant()
   Target High=Constant() {r=1}
  ) {materialFunction=&RemapValueRange}
  n27 = MaterialFunctionCall(
   Plane Normal Axis=Add(A=Multiply(A=@n20, B=@n22), B=@n21.Result)
   Plane Center=Add(A=Multiply(A=@n21.Result, B=@n23), B=StaticSwitchParameter(True=@n24, False=@n25) {group=Cartoon, parameterName=CRTN_UseObjectPos})
   UV Scale=LinearInterpolate(
    A=ScalarParameter() {defaultValue=1.4, group=Cartoon, parameterName=CRTN_VP-Scale-LowFOV}
    B=ScalarParameter() {defaultValue=1.75, group=Cartoon, parameterName=CRTN_VP-Scale-HighFOV}
    Alpha=Saturate(None=@n26.Result)
   )
  ) {materialFunction=&VirtualPlaneCoordinates}
  n28 = TextureSampleParameter2D(UVs=Add(A=Multiply(A=@n16, B=@n18), B=Multiply(A=@n19, B=@n27.UVs))) {
   group=Cartoon
   parameterName=CRTN_CrosshatchTexture
   samplerType=SAMPLERTYPE_Masks
   texture=/Game/Global/Textures/Default/Tileable_Patterns/OakPP_Post
  }
  n29 = OneMinus(None=@n28.R)
  n30 = TextureSampleParameter2D(UVs=@n27.UVs) {
   group=Cartoon
   parameterName=CRTN-DotTexture
   samplerType=SAMPLERTYPE_LinearGrayscale
   texture=&/HalftoneOffset
  }
  n31 = VertexInterpolator(VS=PreSkinnedPosition())
  n32 = Add(
   A=Divide(A=ComponentMask(None=@n31.PS) {b=true}, B=ScalarParameter() {defaultValue=5, group=Cartoon, parameterName=CRTN_LocalMask-Size})
   B=ScalarParameter() {defaultValue=-28, group=Cartoon, parameterName=CRTN_LocalMask-Add}
  )
  n33 = ScalarParameter() {defaultValue=1, group=Cartoon, parameterName=CRTN_LocalMask-Mod}
  n34 = Multiply(A=Saturate(None=Multiply(A=@n32, B=@n33)), B=ScalarParameter() {defaultValue=1, group=Cartoon, parameterName=CRTN_LocalMask-Blend})
  n35 = Constant()
  n36 = Saturate(None=Add(A=@n30.R, B=StaticSwitchParameter(True=@n34, False=@n35) {group=Cartoon, parameterName=CRTN_UseHeadMask}))
  n37 = Multiply(
   A=MaterialFunctionCall!(Diffuse 1, Diffuse 2, Transistion Phase=OneMinus(None=@n15), Height Texture 1=@n29, Height Texture 2=@n36, Contrast) {materialFunction=/Engine/Functions/Engine_MaterialFunctions02/Texturing/HeightLerpWithTwoHeightMaps}.Lerp Alpha No Contrast
   B=ScalarParameter() {defaultValue=0.8, group=Cartoon, parameterName=CRTN_BrightsAmount}
  )
  n38 = Multiply(A=@n37, B=@n11)
  n39 = MaterialFunctionCall(Transistion Phase=OneMinus(None=@n14), Height Texture 1=@n29, Height Texture 2=@n36, Contrast=Constant() {r=1}) {materialFunction=&HeightLerpWithTwoHeightMaps}
  n40 = Add(A=ScalarParameter() {defaultValue=0.5, group=Cartoon, parameterName=CRTN_SL_DarksAmount}, B=@n39.Alpha)
  n41 = Saturate(None=Multiply(A=Add(A=@n38, B=@n11), B=Saturate(None=@n40)))
  n42 = Saturate(None=@n8)
  n43 = Add(A=Multiply(A=Saturate(None=@n9), B=Multiply(A=@n10, B=@n41)), B=@n42)
  n44 = VectorParameter() {parameterName=SL-Basecolor, defaultValue=<rgba 0.07 0.0502 0.0105 1>}
  n45 = MaterialFunctionCall(
   Color_Morning=@n44.RGB
   Color_Day=VectorParameter() {parameterName=SL-Basecolor-Day, defaultValue=<rgba 0.15 0.116 0.048 1>}.RGB
   Color_Evening=@n44.RGB
   Color_Night=@n44.RGB
  ) {materialFunction=&Color_Time_of_Day}
  n46 = Add(A=@n13, B=Multiply(A=ScalarParameter() {defaultValue=0.5, group=Cartoon, parameterName=CRTN_SL_CrosshatchMod}, B=@n29))
  n47 = Multiply(A=@n46, B=ScalarParameter() {defaultValue=8, group=Cartoon, parameterName=CRTN_SL-ShadowReflectAmtMod})
  n48 = ScalarParameter() {defaultValue=-4, group=Cartoon, parameterName=CRTN_SL-ShadowReflectAmtAdd}
  n49 = OneMinus(None=Saturate(None=Add(A=@n47, B=@n48)))
  n50 = Multiply(A=@n42, B=Multiply(A=ScalarParameter() {defaultValue=10, group=Cartoon, parameterName=CRTN_SL-ShadowRefPush}, B=@n49))
  n51 = Max(A=Saturate(None=Multiply(A=@n50, B=@n49)), B=@n41)
  n52 = Saturate(None=Add(A=Multiply(A=@n43, B=@n45.Result), B=Multiply(A=@n51, B=@n45.Result)))
  n53 = Multiply(A=@n46, B=ScalarParameter() {defaultValue=5, group=Cartoon, parameterName=CRTN_SL-LightOnlyMod})
  n54 = ScalarParameter() {defaultValue=-8.5, group=Cartoon, parameterName=CRTN_SL-LightOnlyAdd}
  n55 = Multiply(A=@n45.Result, B=Saturate(None=Add(A=@n53, B=@n54)))
  n56 = Add(
   A=Saturate(None=Add(A=@n52, B=@n55))
   B=Multiply(A=ScalarParameter() {defaultValue=1, group=Cartoon, parameterName=CRTN_EmissiveInputAmount}, B=@n2.EmissiveColor)
  )
  n57 = FunctionInput(Preview=Constant()) {bUsePreviewValueAsDefault=true, inputName=SuperLevelMask, inputType=FunctionInput_Scalar}
  n58 = Saturate(None=Add(A=Multiply(A=@n37, B=@n2.BaseColor), B=@n2.BaseColor))
  n59 = ScalarParameter() {defaultValue=0.2, group=Cartoon, parameterName=CRTN__DarksAmount}
  n60 = ScalarParameter() {defaultValue=0.4, group=Cartoon, parameterName=CRTN_MaskToBrightenAmount}
  n61 = FunctionInput(Preview=Constant()) {bUsePreviewValueAsDefault=true, inputName=UseMaskToBrighten, inputType=FunctionInput_Scalar}
  n62 = Saturate(None=Add(A=Add(A=@n39.Alpha, B=@n59), B=Multiply(A=@n60, B=@n61)))
  n63 = ScalarParameter() {defaultValue=0.4, group=Cartoon, parameterName=CRTN_EmissiveAmount-Morning}
  n64 = ScalarParameter() {defaultValue=0.8, group=Cartoon, parameterName=CRTN_EmissiveAmount-Day}
  n65 = ScalarParameter() {defaultValue=0.3, group=Cartoon, parameterName=CRTN_EmissiveAmount-Eve}
  n66 = ScalarParameter() {defaultValue=0.3, group=Cartoon, parameterName=CRTN_EmissiveAmount-Night}
  n67 = Constant3Vector() {constant=<rgba 1 1 1 0>}
  n68 = VectorParameter() {group=Cartoon, parameterName=CRTN_ColorShift-Day, defaultValue=<rgba 0.88 0.9821 1.1 1>}
  n69 = Multiply(
   A=Multiply(
    A=Multiply(A=@n58, B=@n62)
    B=MaterialFunctionCall(Value_Morning=@n63, Value_Day=@n64, Value_Evening=@n65, Value_Night=@n66) {materialFunction=&Scalar_Time_of_Day}.Result
   )
   B=LinearInterpolate(
    A=Constant() {r=1}
    B=MaterialFunctionCall(Color_Morning=@n67, Color_Day=@n68.RGB, Color_Evening=@n67, Color_Night=@n67) {materialFunction=&Color_Time_of_Day}.Result
    Alpha=@n61
   )
  )
  n70 = SetMaterialAttributes!(
   MaterialAttributes=@n2.MaterialAttributes
   BaseColor=@n1
   EmissiveColor=StaticSwitchParameter(True=LinearInterpolate(A=@n69, B=@n56, Alpha=@n57), False=@n69) {group=SimpleMaterial, parameterName=SM_UseSuperlevellingMask}
   Specular=@n1
   Roughness=Constant() {r=1}
   Metallic=@n1
   ShadingModel=ShadingModel()
  ) {attributeSetTypes [
    #1
    #2
    #3
    #4
    #5
    #6
   ]}
  FunctionOutput(None=@n70)
  n71 = Constant() {r=2.75}
  n72 = Constant3Vector() {desc=SunlightDirectionalVector - from fortnitematerialparameters, constant=<rgba 0 0 1 0>}
  n73 = ComponentMask(None=PixelNormalWS()) {b=true, g=true, r=true}
  n74 = ComponentMask(None=VertexNormalWS()) {b=true, g=true, r=true}
  n75 = Constant() {r=0.25}
  n76 = DotProduct(
   A=ComponentMask(None=ComponentMask(None=@n72) {b=true, g=true, r=true}) {b=true, g=true, r=true}
   B=ShadingPathSwitch(Default=LinearInterpolate(A=@n73, B=@n74, Alpha=@n75), Forward=@n73, Mobile=@n73)
  )
  n77 = Saturate(None=Divide(A=Subtract(A=@n76) {constB=-1}, B=Constant() {r=1.25}))
  n78 = Floor(None=Multiply(A=@n77, B=@n71))
  n79 = Multiply(A=Constant() {r=2}, B=Add(A=Constant() {r=0.1}, B=Divide(A=@n78, B=@n71)))
  n80 = Multiply(A=Constant() {r=1}, B=@n2.BaseColor)
  n81 = OneMinus(None=@n77)
  n82 = Multiply(A=Power(Base=Min(A=@n81, B=@n77), Exp=Constant() {r=1}), B=Constant() {r=1})
  n83 = Time()
  n84 = Multiply(A=Constant() {r=0.46}, B=@n83)
  n85 = Floor(None=Multiply(A=Constant() {r=12}, B=Frac(None=@n84)))
  n86 = Multiply(A=Constant() {r=8.3}, B=@n85)
  n87 = VertexInterpolator(VS=PreSkinnedPosition())
  n88 = ComponentMask(None=MaterialFunctionCall!(
    TextureObject
    TextureSize=Constant() {r=50}
    WorldPosition=Add(A=@n86, B=@n87.PS)
    Export Float 4
    In
    World Space Normal
    ProjectionTransitionContrast
   ) {materialFunction=/Game/Athena/SafeZone/WorldAlignedTexture_ScalePan}.XYZ Texture) {r=true}
  n89 = OneMinus(None=Max(A=@n88, B=Constant()))
  LinearInterpolate(
   A=Saturate(None=Multiply(A=@n79, B=@n80))
   B=Multiply(A=Constant() {r=-5}, B=@n2.BaseColor)
   Alpha=MaterialFunctionCall(Transition Phase=Saturate(None=@n82), Height Texture=@n89, Contrast=Constant()) {materialFunction=&HeightLerp}.Alpha
  )
  n90 = OneMinus(None=Power(Base=@n77, Exp=Constant() {r=2}))
  n91 = Constant() {r=1}
  n92 = Divide(A=@n87.PS, B=Constant() {r=0.8})
  n93 = ComponentMask(None=Constant3Vector() {constant=<rgba 0.5 0.5 0.5 1>}) {b=true, g=true, r=true}
  n94 = Distance(A=Frac(None=Add(A=@n92, B=@n93)), B=Constant3Vector() {constant=<rgba 0.5 0.5 0.5 0>})
  n95 = Saturate(None=Multiply(A=OneMinus(None=@n94), B=Constant() {r=1}))
  MaterialFunctionCall(Transistion Phase=Saturate(None=Multiply(A=@n90, B=@n91)), Height Texture 1=@n89, Height Texture 2=@n95) {materialFunction=&HeightLerpWithTwoHeightMaps}
  n96 = Frac(None=@n92)
  n97 = Constant3Vector() {constant=<rgba 0.5 0.5 0.5 0>}
  n98 = Power(Base=OneMinus(None=Distance(A=@n96, B=@n97)), Exp=Constant() {r=3})
  n99 = Constant() {r=1}
  MaterialFunctionCall(
   Transistion Phase=OneMinus(None=@n77)
   Height Texture 1=@n89
   Height Texture 2=Saturate(None=Multiply(A=@n98, B=@n99))
   Contrast=Constant() {r=0.5}
  ) {materialFunction=&HeightLerpWithTwoHeightMaps}
  n100 = Power(Base=@n77, Exp=Constant() {r=3})
  n101 = Constant() {r=2}
  MaterialFunctionCall(Transition Phase=Saturate(None=Multiply(A=@n100, B=@n101)), Height Texture=@n95, Contrast=Constant() {r=1}) {materialFunction=&HeightLerp}
  n102 = Power(Base=@n77, Exp=Constant() {r=2})
  n103 = Constant() {r=2}
  MaterialFunctionCall(Transition Phase=Saturate(None=Multiply(A=@n102, B=@n103)), Height Texture=@n89, Contrast=Constant() {r=1}) {materialFunction=&HeightLerp}
  Max(A=ScalarParameter() {group=Cartoon, parameterName=CRTN_Crosshatch-Amount}, B=@n28.R)
  Power(Base=@n88, Exp=Constant() {r=1})
  n104 = Multiply(A=Constant() {r=3}, B=@n83)
  n105 = Floor(None=Multiply(A=Constant() {r=3}, B=Frac(None=@n104)))
  Reroute(None=LinearInterpolate(A=Constant() {r=1.48}, B=Constant() {r=1.52}, Alpha=Saturate(None=@n105)))
  n106 = Constant() {r=6600}
  Reroute(None=Add(A=@n87.PS, B=Multiply(A=@n85, B=@n106)))
  n107 = Transform(None=@n5) {transformSourceType=TRANSFORMSOURCE_World, transformType=TRANSFORM_View}
  n108 = Multiply(A=Add(A=Multiply(A=@n107) {constB=0.5}) {constB=0.5}, B=Constant() {r=1})
  n109 = ComponentMask(None=Constant3Vector() {constant=<rgba 0.5 0.5 1 1>}) {b=true, g=true, r=true}
  n110 = Multiply(A=ReflectionVectorWS(CustomWorldNormal=Multiply(A=@n5, B=@n109)), B=Constant())
  TextureSample(UVs=Add(A=ComponentMask(None=@n108) {g=true, r=true}, B=ComponentMask(None=@n110) {g=true, r=true}), Level=Constant()) {
   mipValueMode=TMVM_MipLevel
   samplerSource=SSM_Wrap_WorldGroupSettings
   texture=/Game/Packages/Fortress_SharedMaterials/Textures/T_Ravens_SpecularReflection
  }
  VertexNormalWS()
 ]}
