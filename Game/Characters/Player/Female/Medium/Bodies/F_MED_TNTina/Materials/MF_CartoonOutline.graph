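# Cartoon outline material function for the TNTina character, writing black emissive, opacity and vertex offset on a two-sided shell pushed along normals.
# Backfaces form the outer line, wobbled by a time-stepped flowmap on a camera-facing virtual plane.
# Deferred shading adds scene-depth edge lines, optionally roughness edges; forward and mobile can use a distance-scaled inverted shell instead.
# A hide switch or chrome-strength parameter collapses the shell to the object position.
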MaterialFunction /Game/Characters/Player/Female/Medium/Bodies/F_MED_TNTina/Materials/MF_CartoonOutline
graph {nodes [
  Add(A=ActorPositionWS(), B=Constant3Vector() {constant=<rgba 0 0 80 1>})
  Constant() {r=0.33}
  Constant() {r=12}
  Constant() {r=0.46}
  Constant() {r=1.5}
  Constant() {r=0.75}
  Constant() {r=16}
  Constant() {r=0.25}
  Constant() {r=2}
  Constant() {r=0.75}
  Constant() {r=700}
  Constant() {r=50}
  n1 = Constant()
  n2 = Multiply(A=Subtract(A=WorldPosition().XYZ, B=ObjectPositionWS())) {constB=-1}
  n3 = ScalarParameter() {defaultValue=0.8, group=Cartoon, parameterName=CRTNO_Outline-OuterOpacity}
  n4 = TwoSidedSign()
  n5 = CameraVectorWS()
  n6 = Constant() {r=-1}
  n7 = Saturate(None=DotProduct(A=VertexNormalWS(), B=Multiply(A=@n5, B=@n6)))
  n8 = PixelDepth()
  n9 = ScalarParameter() {defaultValue=50, group=Cartoon, parameterName=CRTNO_Outline-InnerScaleStart}
  n10 = Clamp(None=Divide(A=Subtract(A=@n8.R, B=@n9), B=ScalarParameter() {defaultValue=500, group=Cartoon, parameterName=CRTNO_Outline-InnerScaleDistance}))
  n11 = ObjectPositionWS()
  n12 = CameraPositionWS()
  n13 = OneMinus(None=SphereMask(
    A=Subtract(A=@n11, B=@n12)
    B=Constant()
    Radius=ScalarParameter() {defaultValue=700, group=Cartoon, parameterName=CRTNO_Outline-MobileDistance}
   ) {attenuationRadius=50, hardnessPercent=0})
  n14 = ScalarParameter() {defaultValue=0.46, group=Cartoon, parameterName=CRTNO_Distort-OffsetSpeed}
  n15 = Time()
  n16 = Multiply(A=ScalarParameter() {defaultValue=13, group=Cartoon, parameterName=CRTNO_Distort-NumberOffsets}, B=Frac(None=Multiply(A=@n14, B=@n15)))
  n17 = Constant3Vector() {constant=<rgba 1 1 0 0>}
  n18 = MaterialFunctionCall() {materialFunction=&CameraDirectionVector}
  n19 = Multiply(A=Constant() {desc=Needs at least a small amount of this or it will break, r=0.25}, B=Normalize(VectorInput=Multiply(A=@n17, B=@n18.Result)))
  n20 = Multiply(A=@n18.Result, B=ScalarParameter() {defaultValue=70, group=Cartoon, parameterName=CRTNO_VP-CameraVectorMin})
  n21 = Add(
   A=Multiply(A=ScalarParameter() {defaultValue=0.33, group=Cartoon, parameterName=CRTNO_DistortOffsetFraction}, B=Floor(None=@n16))
   B=MaterialFunctionCall(
    Plane Normal Axis=Add(A=@n19, B=@n18.Result)
    Plane Center=Add(A=@n20, B=@n11)
    UV Scale=ScalarParameter() {defaultValue=50, group=Cartoon, parameterName=CRTNO_VP-Scale}
   ) {materialFunction=&VirtualPlaneCoordinates}.UVs
  )
  n22 = Multiply(
   A=ComponentMask(None=TextureSample(UVs=@n21) {samplerType=SAMPLERTYPE_LinearColor, texture=/Game/Global/Textures/Default/Flowmaps/T_TorusFlowmap}.RGB) {g=true, r=true}
   B=ScalarParameter() {defaultValue=0.6, group=Cartoon, parameterName=CRTNO_DistortionAmount}
  )
  n23 = LinearInterpolate(
   A=LinearInterpolate(
    A=ScalarParameter() {defaultValue=-6, group=Cartoon, parameterName=CRTNO_LineMask-Near}
    B=ScalarParameter() {defaultValue=-1, group=Cartoon, parameterName=CRTNO_LineMask-Far}
    Alpha=ShadingPathSwitch(Default=@n10, Forward=@n13, Mobile=@n13)
   )
   B=ScalarParameter() {defaultValue=1.5, group=Cartoon, parameterName=CRTNO_LineMaskB}
   Alpha=TextureSample(UVs=Add(A=@n7, B=@n22)) {texture=/Game/Global/Textures/Default/Masks_and_Gradients/gradient_stasis}.R
  ) {constA=-2, constB=2}
  n24 = Multiply(A=Saturate(None=OneMinus(None=@n4)), B=Multiply(A=OneMinus(None=@n7), B=Saturate(None=@n23)))
  n25 = PixelDepth()
  n26 = SceneTexture() {sceneTextureId=PPI_SceneDepth}
  n27 = ScalarParameter() {defaultValue=3, group=Cartoon, parameterName=CRTNO_Outline-InnerMaskBias}
  n28 = Divide(A=ObjectRadius(), B=ScalarParameter() {defaultValue=97, parameterName=ObjectScale})
  n29 = Clamp(None=Add(A=Subtract(A=@n25.R, B=@n26.Color), B=Multiply(A=@n27, B=@n28)))
  n30 = Constant() {r=1}
  n31 = ComponentMask(None=ScreenPosition().ViewportUV) {g=true, r=true}
  n32 = Subtract(A=Divide(A=ViewSize(), B=Constant() {r=320}), B=Constant() {r=2}) {constB=0}
  n33 = LinearInterpolate(
   A=ScalarParameter() {defaultValue=2, group=Cartoon, parameterName=CRTNO_Outline-InnerSize1}
   B=ScalarParameter() {defaultValue=0.75, group=Cartoon, parameterName=CRTNO_Outline-InnerSize2}
   Alpha=@n10
  )
  n34 = Multiply(A=@n33, B=ScalarParameter() {defaultValue=2, group=Cartoon, parameterName=CRTNO_Outline-InnerSizeMobileMultiplier})
  n35 = Multiply(
   A=Clamp(None=Ceil(None=@n32)) {maxDefault=2, minDefault=1}
   B=Multiply(A=SceneTexelSize(), B=ShadingPathSwitch(Default=@n33, Forward=@n34, Mobile=@n34))
  ) {constA=2}
  n36 = Constant()
  n37 = AppendVector(A=ComponentMask(None=@n35) {desc=x, r=true}, B=@n36)
  n38 = Add(A=@n31, B=@n37) {desc=x+}
  n39 = AppendVector(A=@n36, B=ComponentMask(None=@n35) {desc=y, g=true}) {desc=y+}
  n40 = Add(A=@n31, B=@n39) {desc=y+}
  n41 = Constant() {r=-1}
  n42 = Add(A=@n31, B=Multiply(A=@n37, B=@n41)) {desc=x-}
  n43 = Add(A=@n31, B=Multiply(A=@n41, B=@n39)) {desc=y-}
  n44 = ComponentMask(None=MaterialFunctionCall!(
    Base=@n26.Color
    BaseSampleBias=Multiply(A=@n30, B=@n28)
    Scale=Constant() {r=4}
    Sample1=SceneTexture(UVs=@n38) {sceneTextureId=PPI_SceneDepth}.Color
    Sample2=SceneTexture(UVs=@n40) {sceneTextureId=PPI_SceneDepth}.Color
    Sample3=SceneTexture(UVs=@n42) {sceneTextureId=PPI_SceneDepth}.Color
    Sample4=SceneTexture(UVs=@n43) {sceneTextureId=PPI_SceneDepth}.Color
    Divisor=Constant() {r=-16}
    UseCenteredLines
    UseRGB
   ) {materialFunction=/Game/Packages/jl_tests/combine4Samples}.Result) {r=true}
  n45 = ComponentMask(None=MaterialFunctionCall(
    Base=SceneTexture() {sceneTextureId=PPI_Roughness}.Color
    BaseSampleBias=Constant()
    Scale=Constant() {r=4}
    Sample1=SceneTexture(UVs=@n38) {sceneTextureId=PPI_Roughness}.Color
    Sample2=SceneTexture(UVs=@n40) {sceneTextureId=PPI_Roughness}.Color
    Sample3=SceneTexture(UVs=@n42) {sceneTextureId=PPI_Roughness}.Color
    Sample4=SceneTexture(UVs=@n43) {sceneTextureId=PPI_Roughness}.Color
    Divisor=ScalarParameter() {defaultValue=-0.04, group=RoughnessGBufferLines, parameterName=InnerLinesDivisor}
   ) {materialFunction=&combine4Samples}.Result) {r=true}
  n46 = StaticSwitchParameter(True=ShadingPathSwitch(Default=Max(A=@n44, B=@n45), Forward=@n44, Mobile=@n44), False=@n44) {group=RoughnessGBufferLines, parameterName=UseRoughnessGBufferOutlines, sortPriority=0}
  n47 = Multiply(A=Saturate(None=Multiply(A=@n29, B=@n46)), B=ScalarParameter() {defaultValue=0.7, group=Cartoon, parameterName=CRTNO_Outline-InnerAmount})
  n48 = TwoSidedSign()
  n49 = StaticSwitchParameter(True=Saturate(None=OneMinus(None=@n48)), False=Constant()) {
   desc=Uses bodybag as inverted shell for mobile outlines. needs two sided material
   group=Cartoon MobileShell
   parameterName=UseInvertedShellOutlineOnMobile
   sortPriority=0
  }
  n50 = Saturate(None=Add(A=Multiply(A=@n3, B=@n24), B=ShadingPathSwitch(Default=@n47, Forward=@n49, Mobile=@n49)))
  n51 = VertexNormalWS()
  n52 = ScalarParameter() {defaultValue=0.25, group=Cartoon, parameterName=CRTNO_Outline-ShellDistance}
  n53 = Divide(
   A=Subtract(
    A=MaterialFunctionCall!(World Position) {materialFunction=/Engine/Functions/Engine_MaterialFunctions02/Utility/DepthFromWorldPosition}.Result
    B=ScalarParameter() {defaultValue=50, group=Cartoon MobileShell, parameterName=CRTNO_Outline-MobileInvertedShellScaleStart}
   )
   B=ScalarParameter() {defaultValue=1500, group=Cartoon MobileShell, parameterName=CRTNO_Outline-MobileInvertedShellScaleDistance}
  )
  n54 = StaticSwitchParameter(
   True=LinearInterpolate(
    A=ScalarParameter() {defaultValue=0.1, group=Cartoon MobileShell, parameterName=CRTNO_Outline-MobileInvertedShellDistance}
    B=ScalarParameter() {defaultValue=1, group=Cartoon MobileShell, parameterName=CRTNO_Outline-MobileInvertedShellDistanceFar}
    Alpha=Clamp(None=@n53)
   )
   False=@n52
  ) {
   desc=Uses bodybag as inverted shell for mobile outlines. needs two sided material
   group=Cartoon MobileShell
   parameterName=UseInvertedShellOutlineOnMobile
  }
  n55 = ShadingPathSwitch(Default=@n52, Forward=@n54, Mobile=@n54)
  n56 = SetMaterialAttributes!(
   MaterialAttributes
   EmissiveColor=Constant()
   Opacity=@n50
   WorldPositionOffset=LinearInterpolate(
    A=Multiply(A=@n51, B=@n55)
    B=@n2
    Alpha=ScalarParameter() {group=S22ChromeCharacters, parameterName=S22ChromeCharacters_strength, sliderMax=1}
   )
   OpacityMask=@n50
  ) {attributeSetTypes [
    #1
    #2
    #3
    #4
   ]}
  FunctionOutput(None=StaticSwitchParameter(
    True=SetMaterialAttributes!(MaterialAttributes, EmissiveColor=@n1, Opacity=@n1, WorldPositionOffset=@n2) {attributeSetTypes [#1, #2, #3]}
    False=@n56
   ) {group=Cartoon, parameterName=CRTNO_HideCartoonOutline})
  n57 = MaterialFunctionCall(Float2=ViewSize()) {materialFunction=&BreakOutFloat2Components}
  n58 = AppendVector(A=Constant() {r=1}, B=Divide(A=@n57.G, B=@n57.G))
  n59 = Multiply(A=ConstantBiasScale(None=ScreenPosition().ViewportUV) {bias=-0.5, scale=2}, B=@n58)
  n60 = ActorPositionWS()
  n61 = Multiply(
   A=Constant2Vector() {g=-0.5, r=0.5}
   B=MaterialFunctionCall(World Position=@n60) {materialFunction=&TransformToClipSpace}.Clip Space XY
  )
  n62 = Constant() {r=1}
  n63 = ComponentMask(None=TransformPosition(None=@n60) {transformSourceType=TRANSFORMPOSSOURCE_World, transformType=TRANSFORMPOSSOURCE_View}) {b=true}
  MaterialFunctionCall(
   UVs=Subtract(A=ConstantBiasScale(None=@n59), B=Multiply(A=@n58, B=@n61))
   Texture Scale=Multiply(A=Divide(A=@n62, B=@n63), B=ScalarParameter() {defaultValue=16, group=Cartoon, parameterName=CRTNO_DistortScale})
  ) {materialFunction=&ScaleUVsByCenter}
  ObjectRadius()
  MaterialExpressionScreenPosition_1 = ScreenPosition()
  n64 = Custom!(UV=@MaterialExpressionScreenPosition_1.ViewportUV) {
   code="return Texture2DSampleLevel(SceneTexturesStruct.GBufferBTexture, SceneTexturesStruct.GBufferBTextureSampler, UV.xy, 0).b;"
   outputType=CMOT_Float1
   inputs [{
     inputName=UV
     input {expression=&self:MaterialExpressionScreenPosition_1}
    }]
  }
  MaterialExpressionConstant2Vector_1 = Constant2Vector()
  n65 = Custom!(p=@MaterialExpressionScreenPosition_1.ViewportUV, o=@MaterialExpressionConstant2Vector_1) {
   code="return SceneTexturesStruct.GBufferBTexture.GatherBlue(SceneTexturesStruct.GBufferBTextureSampler, p.xy, o.xy);"
   outputType=CMOT_Float4
   inputs [
    {
     inputName=p
     input {expression=&self:MaterialExpressionScreenPosition_1}
    }
    {
     inputName=o
     input {expression=&self:MaterialExpressionConstant2Vector_1}
    }
   ]
  }
  n66 = MaterialFunctionCall(Float4=@n65) {materialFunction=&BreakOutFloat4Components}
  RayTracingQualitySwitch(
   Normal=MaterialFunctionCall(
    Base=@n64
    BaseSampleBias=Constant()
    Scale=Constant() {r=1}
    Sample1=@n66.R
    Sample2=@n66.R
    Sample3=@n66.R
    Sample4=@n66.R
    Divisor=ScalarParameter() {defaultValue=0.1, parameterName=Div}
   ) {materialFunction=&combine4Samples}.Result
   RayTraced=Constant()
  )
  VectorParameter() {group=Cartoon, parameterName=CRTNO_Outline-Color, defaultValue=<rgba 0 0 0 1>}
 ]}
